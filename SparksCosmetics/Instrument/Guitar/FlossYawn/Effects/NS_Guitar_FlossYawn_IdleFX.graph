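# Niagara idle effect for the FlossYawn guitar cosmetic: an icy blue glow, spiraling ribbon, snowflakes and sparkles.
# Glow, Snowflakes and Sparkles follow skeletal mesh sockets, Line streaks spawn from snowflake particles, and one procedural sine/cosine ribbon gets four renderers.
# Audio amplitude, or fake reactivity in preview, drives brightness and hides flakes, sparkles and lines at zero.
# Time of day, platform and front-end preview also adjust intensity; DebugValues is disabled.
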NiagaraSystem /SparksCosmetics/Instrument/Guitar/FlossYawn/Effects/NS_Guitar_FlossYawn_IdleFX
dependencies {
 usedDataInterfaces [
  &NiagaraDataInterfaceCurve
  &NiagaraDataInterfaceColorCurve
  &NiagaraDataInterfaceArrayFloat3
  &NiagaraDataInterfaceArrayInt32
  &NiagaraDataInterfacePlatformSet
  &NiagaraDataInterfaceSpriteRendererInfo
  &NiagaraDataInterfaceParticleRead
  &NiagaraDataInterfaceArrayFloat
  &NiagaraDataInterfaceSkeletalMesh
 ]
 usedDynamicInputs [
  &MakeCustomLinearColorFromBool
  &RandomBool
  &MakeFloatFromVector
 ]
 usedModules [
  &self:SystemSpawnScript.SystemScriptSource.SystemScriptGraph.NiagaraNodeAssignment_2.SetVariables_#1
  &SystemState
  &self:SystemSpawnScript.SystemScriptSource.SystemScriptGraph.NiagaraNodeAssignment_0.SetVariables_#2
  &self:SystemSpawnScript.SystemScriptSource.SystemScriptGraph.NiagaraNodeAssignment_1.SetVariables_#3
  &self:SystemSpawnScript.SystemScriptSource.SystemScriptGraph.NiagaraNodeAssignment_5.SetVariables_#4
  &Timeline
  &self:SystemSpawnScript.SystemScriptSource.SystemScriptGraph.NiagaraNodeAssignment_4.SetVariables_#5
  &self:SystemSpawnScript.SystemScriptSource.SystemScriptGraph.NiagaraNodeAssignment_3.SetVariables_#6
  &InterpolateOverTime
 ]
}
emitters {
 DebugValues {
  bEnabled=false
  emitterUpdateScript [
   &EmitterState()
   &SpawnBurst_Instantaneous(Use Spawn Probability=false, UseLoopCountLimit=true)
  ]
  particleSpawnScript [&InitializeParticle(
    Color Mode=::NewEnumerator0
    Position Offset=<0 0 100>
    Sprite Size Mode=::NewEnumerator3
    Uniform Sprite Size=90
    UsePositionOffset=true
    Write Material Random=true
   )]
  particleUpdateScript [
   &ParticleState(Kill Particles When Lifetime Has Elapsed=false, Let Infinitely Lived Particles Die When Emitter Deactivates=true)
   &SolveForcesAndVelocity(Clamp Velocity=false, Limit Acceleration=false)
   &DynamicMaterialParameters(
    Index 0 Param 1=&MakeCustomFloatFromBool(Bool=true, True Float=1, False Float=0)
    Module.Index0Float0Write=true
    Module.Index0Float1Write=true
    Module.Index0Float2Write=true
    Module.Index0Float3Write=true
   )
  ]
  rendererClasses [&NiagaraSpriteRendererProperties]
  renderers [{rendererClass=&NiagaraSpriteRendererProperties}]
 }
 Glow {
  emitterUpdateScript [
   &EmitterState(
    Inactive Response=::NewEnumerator1
    Life Cycle Mode=::NewEnumerator1
    Loop Behavior=::NewEnumerator1
    Loop Duration Mode=::NewEnumerator1
    UseLoopDelay=false
   )
   &SpawnBurst_Instantaneous(Spawn Count=12, Use Spawn Probability=true, UseLoopCountLimit=false)
  ]
  particleSpawnScript [
   &PartitionParticles(Number Of Partitions=12)
   &InitializeParticle(
    Color=<rgba 0.0526 0.2913 0.5 3>
    Sprite Size Mode=::NewEnumerator3
    Uniform Sprite Size=&Multiply_Float(
     A=&SelectFloatFromArray(
      Direct Array Index=$Particles.PartitionParticles.Partition
      Float Selection Array=DataInterface() {propertyValues {FloatData [50,50,50,50,30,30,30,30,30,30,30,40]}}
     )
     B=$System.MeshScale
    )
    UsePositionOffset=false
    Write Material Random=true
   )
  ]
  particleUpdateScript [
   &ParticleState(Kill Particles When Lifetime Has Elapsed=false, Let Infinitely Lived Particles Die When Emitter Deactivates=true)
   &ScaleSpriteSize(Uniform Curve Sprite Scale=DataInterface() {propertyValues {Curve {keys [{tangentMode=RCTM_User, time=0, value=1}, {tangentMode=RCTM_User, time=1, value=1}]}}}) {enabled=false}
   &SkeletalMeshLocation(
    Bone / Socket Index=&SelectIntFromArray(
     Direct Array Index=$Particles.PartitionParticles.Partition
     Int Selection Array=DataInterface() {propertyValues {IntData [0,0,0,0,1,1,1,1,1,1,1,2]}}
    )
    MaskSpawnGroup=false
    Mesh Sampling Type=::NewEnumerator5
    Sampled Position Offset=&Multiply_VectorByFloat(
     Vector=&SelectVectorFromArray(
      Direct Array Index=$Particles.PartitionParticles.Partition
      Vector Selection Array=DataInterface() {propertyValues {FloatData [
         <0 -2 -20>
         <-9 -2 -10.5>
         <9 -2 -10.5>
         <0 -2 0>
         <0 0 -26.8751>
         <0 0 -18.8359>
         <0 0 -9.4117>
         <0 0 -0.8769>
         <0 0 7.5969>
         <0 0 16.1272>
         <0 0 24.8249>
         <-1.2739 0 8.1103>
        ]}}
     )
     Float=$System.MeshScale
    )
    Skeletal Mesh=$System.Skeletal Mesh
    Socket Sampling Mode=::NewEnumerator5
   )
   &SolveForcesAndVelocity(Clamp Velocity=false, Limit Acceleration=false)
   &DynamicMaterialParameters(
    Index 0 Param 1=3.5
    Index 0 Param 2=5
    Module.Index0Param0Write=true
    Module.Index0Param1Write=true
    Module.Index0Param2Write=true
    Module.Index0Param3Write=true
   )
   &ScaleColor(
    Scale Alpha=&Multiply_Float(
     A=&Multiply_Float(
      A=&SelectFloatFromArray(
       Direct Array Index=$Particles.PartitionParticles.Partition
       Float Selection Array=DataInterface() {propertyValues {FloatData [0.1,0.1,0.1,0.1,0.1,0.1,0.1,0.1,0.1,0.1,0.12,0.15]}}
      )
      B=&FN_PickFloatDependOnPlatform(TargetPlatformActive=1, TargetPlatformNoActive=0.75)
     )
     B=&Lerp_Float(
      A=&Lerp_Float(A=1, B=1.75, Alpha=$NPC.FortniteNPC.EventAudio_Amplitude)
      B=&Lerp_Float(A=1, B=1.75, Alpha=$System.FEFakeReactivity)
      Alpha=$System.UsePreview
     )
    )
    Scale RGB=&VectorFromFloat(Value=&Multiply_Float(
      A=&MakeCustomFloatFromBool(Bool=$User.bIsFrontEndPreview, True Float=0.7, False Float=1.2)
      B=&FloatFromCurve(
       FloatCurve=DataInterface() {propertyValues {Curve {keys [<at 0 0.7>,<at 0.15 1>,<at 0.696 1>,<at 0.9 0.7>]}}}
       CurveIndex=Linked() {linkedVariable {name=NPC.FortniteNPC.TimeOfDay_0-1, type=NiagaraFloat}}
      )
     ))
    ScaleA=true
    ScaleRGB=true
   )
  ]
  rendererClasses [&NiagaraSpriteRendererProperties, &NiagaraMeshRendererProperties]
  renderers [
   {rendererClass=&NiagaraSpriteRendererProperties}
   {rendererClass=&NiagaraMeshRendererProperties, rendererIndex=1}
  ]
 }
 Line {
  emitterUpdateScript [
   &EmitterState(Inactive Response=::NewEnumerator1, Life Cycle Mode=::NewEnumerator1, Loop Duration=1, UseLoopDelay=false)
   &SpawnParticlesFromOtherEmitter(
    Attribute Reader=DataInterface() {propertyValues {EmitterBinding {bindingMode=Other, emitterName=SnowFlakes}}}
    Spawn Rate=&Multiply_Float(A=15, B=$System.DistanceSpawn)
    Use Maximum Age=false
    Use Minimum Age=false
   )
  ]
  particleSpawnScript [
   &InitializeParticle(
    Lifetime Max=0.35
    Lifetime Min=0.15
    Lifetime Mode=::NewEnumerator1
    Mesh Scale Max=<0.5 0.5 10>
    Mesh Scale Min=<0.1 0.1 5>
    Mesh Scale Mode=::NewEnumerator2
    Sprite Size Max=<1.5 16>
    Sprite Size Min=<0.5 10>
    Sprite Size Mode=::NewEnumerator2
    UsePositionOffset=false
    Write Material Random=true
   )
   &SampleParticlesFromOtherEmitter(
    Age Sampling:ENiagara_DirectReadSamplingMode=::NewEnumerator0
    Lifetime Application:ENiagara_LifetimeInheritanceOptions=::NewEnumerator0
    Lifetime Sampling:ENiagara_DirectReadSamplingMode=::NewEnumerator0
    Velocity Application:ENiagara_DirectReadApplicationMode=::NewEnumerator0
    Velocity Sampling=::NewEnumerator0
   )
   &ShapeLocation(MaskSpawnGroup=false, Override Seed=false, Shape Origin=$Particles.Position, Sphere Radius=4, UseOwnerScale=true) {module=ShapeLocation001}
   &CameraOffset(Camera Offset Amount=15)
  ]
  particleUpdateScript [
   &ParticleState()
   &SolveForcesAndVelocity(Clamp Velocity=false, Limit Acceleration=false)
   /SetVariables(Particles.Color_glow:LinearColor=&MakeLinearColorFromVectorAndFloat(
     "Vector (RGB)"=&MakeVectorFromLinearColorRGB(Color=&ColorFromCurve(ColorCurve=DataInterface() {propertyValues {
         AlphaCurve {keys []}
         BlueCurve {keys [
           {interpMode=RCIM_Cubic, tangentMode=RCTM_User, time=0.1306, value=0.8473}
           {interpMode=RCIM_Cubic, tangentMode=RCTM_User, time=0.3325, value=1}
           {interpMode=RCIM_Cubic, tangentMode=RCTM_User, time=0.474, value=1.5}
           {interpMode=RCIM_Cubic, tangentMode=RCTM_User, time=0.5973, value=1}
           {interpMode=RCIM_Cubic, tangentMode=RCTM_User, time=0.7523, value=0.8473}
          ]}
         GreenCurve {keys [
           {interpMode=RCIM_Cubic, tangentMode=RCTM_User, time=0.1306, value=1}
           {interpMode=RCIM_Cubic, tangentMode=RCTM_User, time=0.3325, value=0.5388}
           {interpMode=RCIM_Cubic, tangentMode=RCTM_User, time=0.474, value=1.2267}
           {arriveTangent=3.4579, interpMode=RCIM_Cubic, leaveTangent=3.4579, tangentMode=RCTM_User, time=0.5973, value=0.4231}
           {interpMode=RCIM_Cubic, tangentMode=RCTM_User, time=0.7523, value=1}
          ]}
         RedCurve {keys [
           {interpMode=RCIM_Cubic, tangentMode=RCTM_User, time=0.1306, value=0}
           {interpMode=RCIM_Cubic, tangentMode=RCTM_User, time=0.3325, value=0.1031}
           {interpMode=RCIM_Cubic, tangentMode=RCTM_User, time=0.474, value=0.6135}
           {interpMode=RCIM_Cubic, tangentMode=RCTM_User, time=0.5973, value=0}
           {interpMode=RCIM_Cubic, tangentMode=RCTM_User, time=0.7523, value=0}
          ]}
        }}))
     "Float (Alpha)"=&FloatFromCurve(
      FloatCurve=DataInterface() {propertyValues {Curve {keys [<at 0.5 1>,<at 1 0>]}}}
      Scale Curve=&MakeCustomFloatFromBool(
       Bool=&SetBoolByFloatComparison(
        A=&Max_Float(
         A:NiagaraFloat=&Lerp_Float(A=0, B=$System.FEFakeReactivity, Alpha=$System.UsePreview)
         B:NiagaraFloat=$NPC.FortniteNPC.EventAudio_Amplitude
        )
        B=0
       )
       True Float=&MakeCustomFloatFromBool(
        Bool=$User.bIsFrontEndPreview
        True Float=&Lerp_Float(
         A=&Lerp_Float(A=0.7, B=1.4, Alpha=$NPC.FortniteNPC.EventAudio_Amplitude)
         B=&Lerp_Float(A=0.7, B=1.4, Alpha=$System.FEFakeReactivity)
         Alpha=$System.UsePreview
        )
        False Float=&Lerp_Float(A=1, B=2, Alpha=$NPC.FortniteNPC.EventAudio_Amplitude)
       )
       False Float=0
      )
     )
    )) {isSetParameters=true}
  ]
  rendererClasses [&NiagaraSpriteRendererProperties]
  renderers [{rendererClass=&NiagaraSpriteRendererProperties}]
 }
 NE_SpiralRibbon {
  emitterSpawnScript [/SetVariables(Emitter.ParticlePerRibbon=35, Emitter.RibbonNumber=1) {isSetParameters=true}]
  emitterUpdateScript [
   &EmitterState(
    Inactive Response=::NewEnumerator1
    Life Cycle Mode=::NewEnumerator1
    Loop Behavior=::NewEnumerator1
    Loop Duration Mode=::NewEnumerator1
    UseLoopDelay=false
   )
   &SpawnBurst_Instantaneous(
    Spawn Count=&Multiply_Int(A=$Emitter.ParticlePerRibbon, B=$Emitter.RibbonNumber)
    Use Spawn Probability=false
    UseLoopCountLimit=false
   )
  ]
  particleSpawnScript [
   &InitializeParticle(Lifetime=5, Ribbon Width Mode=::NewEnumerator1, Sprite Size Mode=::NewEnumerator3, UsePositionOffset=false, Write Material Random=true)
   &PartitionParticles(
    ID to Partition:NiagaraInt32=&ReturnExecIndex()
    Number Of Partitions=$Emitter.RibbonNumber
    Particles Per Partition:NiagaraInt32=$Emitter.ParticlePerRibbon
    Partition Mode:ENiagara_PartitionMode=::NewEnumerator1
   )
   /SetVariables(Particles.NormalizedRibbonPos:NiagaraFloat=&OneMinusFloat(Float=$Particles.PartitionParticles.Normalized Index In Partition)) {isSetParameters=true}
   /SetVariables(Particles.SpiralFreq=0.34, Particles.SpiralStrength=1) {isSetParameters=true}
   /SetVariables(Particles.RibbonU0Override=&OneMinusFloat(Float=$Particles.NormalizedRibbonPos)) {isSetParameters=true}
   /SetVariables() {isSetParameters=true}
  ]
  particleUpdateScript [
   &ParticleState(Kill Particles When Lifetime Has Elapsed=false)
   &SkeletalMeshLocation(
    Bone Sampling Mode=::NewEnumerator4
    MaskSpawnGroup=false
    Position Sampling=::NewEnumerator2
    Skeletal Mesh=$System.Skeletal Mesh
    Use Sampled Mesh Attribute as Ribbon ID=false
   )
   /SetVariables(Particles.GoalPos:NiagaraPosition=&AddVectorToPosition(
     Position=$Particles.SkeletalMeshLocation.SampledPosition
     Vector=&Lerp_Vector(A=<0 0 -25>, B=<0 0 70>, Alpha=$Particles.NormalizedRibbonPos)
    )) {isSetParameters=true}
   /SetVariables(Particles.SpiralStrength:NiagaraFloat=&Multiply_Float(
     A=&Multiply_Float(
      A=&Multiply_Float(A=2, B=&Lerp_Float(A=1, B=1.25, Alpha=$System.InterpolateOverTime.InterpolateOverTime_Float.Moving Average))
      B=&Waveform(
       Global Phase=&Frac_Float(Float=&Add_Float(A=$Engine.Time, B=$Particles.NormalizedRibbonPos))
       "Waveform [1]"=::NewEnumerator12
       "[1] Frequency"=0.25
       "[1] Amplitude Min/Max"=<0.75 1.25>
      )
     )
     B=&FloatFromCurve(
      FloatCurve=DataInterface() {propertyValues {Curve {keys [
          {arriveTangent=1.1918, interpMode=RCIM_Cubic, leaveTangent=1.1918, leaveTangentWeight=0.0685, tangentMode=RCTM_User, time=0, value=0}
          {
           arriveTangent=6.907
           arriveTangentWeight=0.044
           interpMode=RCIM_Cubic
           leaveTangent=6.907
           leaveTangentWeight=0.3143
           tangentMode=RCTM_User
           time=0.2699
           value=0.6053
          }
          {arriveTangent=-0.5039, interpMode=RCIM_Cubic, leaveTangent=-0.5039, tangentMode=RCTM_User, time=0.4334, value=1.2578}
          {
           arriveTangent=-2.6387
           arriveTangentWeight=0.6289
           interpMode=RCIM_Cubic
           leaveTangent=-12.1423
           leaveTangentWeight=0.3328
           tangentMode=RCTM_Break
           time=0.7089
           value=1.0219
          }
          {arriveTangentWeight=0.1228, interpMode=RCIM_Cubic, tangentMode=RCTM_User, time=1, value=0}
         ]}}}
      CurveIndex=$Particles.NormalizedRibbonPos
     )
    )) {isSetParameters=true}
   /SetVariables(Particles.SpiralTime:NiagaraFloat=&Add_Float(
     A=&Add_Float(A=&Multiply_Float(A=$System.Age, B=-0.09), B=$Particles.NormalizedRibbonPos)
     B=&Multiply_FloatByInt(
      Float=&Divide_Float(A=$Particles.SpiralFreq, B=&MakeFloatFromInt(Int=$Emitter.RibbonNumber))
      Integer=$Particles.PartitionParticles.Partition
     )
    )) {isSetParameters=true}
   /SetVariables(Particles.SpiralOffset:Vector3f=&MakeVector(
     X=&Sine(
      Normalized Angle=$Particles.SpiralTime
      Period=$Particles.SpiralFreq
      Scale=&Multiply_Float(A=$Particles.SpiralStrength, B=2)
      Bias=0
     )
     Y=&Cosine(
      Normalized Angle=$Particles.SpiralTime
      Period=$Particles.SpiralFreq
      Scale=&Multiply_Float(A=$Particles.SpiralStrength, B=1.25)
      Bias=0
     )
     Z=0
    )) {isSetParameters=true}
   &InheritSourceMovement(
    Applied Position Scale=&VectorFromFloat(Value=&Multiply_Float(
      A=-0.225
      B=&FloatFromCurve(
       FloatCurve=DataInterface() {propertyValues {Curve {keys [
           {arriveTangent=0.3323, interpMode=RCIM_Cubic, leaveTangent=0.3323, tangentMode=RCTM_User, time=0, value=0}
           <at 0.5 1 RCIM_Cubic>
           {arriveTangent=0.1272, interpMode=RCIM_Cubic, leaveTangent=0.1272, tangentMode=RCTM_User, time=1, value=0}
          ]}}}
       CurveIndex=$Particles.NormalizedRibbonPos
      )
     ))
    ApplyForce=false
    ApplyPosition=true
    ApplyPositionFalloff=false
    Source Velocity:Vector3f=&TransformVector(Vector=$Engine.Owner.Velocity, Source Space=::World, Destination Space=::Local)
   )
   &SolveForcesAndVelocity(Clamp Velocity=false, Limit Acceleration=false)
   &ScaleRibbonWidth(Ribbon Width Scale=&Multiply_Float(
     A=&FloatFromCurve(
      FloatCurve=DataInterface() {propertyValues {Curve {keys [
          {arriveTangent=0.6169, interpMode=RCIM_Cubic, leaveTangent=0.6169, tangentMode=RCTM_User, time=0, value=0.2}
          {arriveTangent=1.3696, interpMode=RCIM_Cubic, leaveTangent=1.3696, time=0.2651, value=0.9935}
          {arriveTangent=-0.1475, interpMode=RCIM_Cubic, leaveTangent=-0.1475, time=0.6024, value=1.0251}
          {arriveTangent=-1.8392, interpMode=RCIM_Cubic, leaveTangent=-1.8392, time=0.8012, value=0.9144}
          {arriveTangent=-0.6127, interpMode=RCIM_Cubic, leaveTangent=-0.6127, tangentMode=RCTM_User, time=0.9967, value=0.3}
         ]}}}
      CurveIndex=$Particles.NormalizedRibbonPos
      Scale Curve=&Multiply_Float(A=1.1, B=$System.MeshScale)
     )
     B=&Lerp_Float(A=1, B=1.25, Alpha=$System.InterpolateOverTime.InterpolateOverTime_Float.Moving Average)
    ))
   /SetVariables(Particles.Position=&AddVectorToPosition(Position=$Particles.GoalPos, Vector=$Particles.SpiralOffset)) {isSetParameters=true}
   /SetVariables(Particles.RibbonFacing=&NormalizeVector(
     Vector To Normalize=&TransformVector(Vector=$Particles.SpiralOffset, Destination Space=::Simulation)
     Safe Normalize=true
     Safe Fallback Threshold=0
     Safe Fallback Vector=<0 0 0.001>
    )) {isSetParameters=true}
   /SetVariables(Particles.BackSpiralColor:LinearColor=&MultiplyLinearColor(
     Linear Color=$Particles.Color
     Scale Factor=&MakeVector4_FromLinearColor(LinearColor=<rgba 0.1224 0.5 0.5 1>)
    )) {isSetParameters=true}
   /SetVariables(Particles.BackRibbonFacing:Vector3f=&NormalizeVector(
     Vector To Normalize=&TransformVector(Vector=&Multiply_VectorByFloat(Vector=$Particles.SpiralOffset, Float=-1), Destination Space=::Simulation)
     Safe Normalize=true
     Safe Fallback Threshold=0
     Safe Fallback Vector=<0 0 0.001>
    )) {isSetParameters=true}
   /SetVariables(Particles.GlowWidth:NiagaraFloat=&Multiply_Float(A=$Particles.RibbonWidth, B=1.75)) {isSetParameters=true}
   /SetVariables(Particles.Color_glow:LinearColor=&Multiply_LinearColorByFloat(
     Linear Color=<rgba 0.0931 0.3739 1 1>
     Float=&Multiply_Float(
      A=&Lerp_Float(
       A=&Lerp_Float(A=1, B=1.5, Alpha=$NPC.FortniteNPC.EventAudio_Amplitude)
       B=&Lerp_Float(A=1, B=1.25, Alpha=$System.FEFakeReactivity)
       Alpha=$System.UsePreview
      )
      B=&MakeCustomFloatFromBool(Bool=$User.bIsFrontEndPreview, True Float=0.8, False Float=1)
     )
    )) {isSetParameters=true}
   /SetVariables(Particles.Color_star:LinearColor=&Multiply_LinearColorByFloat(
     Linear Color=<rgba 40 40 40 1>
     Float=&Lerp_Float(A=1, B=2, Alpha=$System.InterpolateOverTime.InterpolateOverTime_Float.Moving Average)
    )) {isSetParameters=true}
   /SetVariables(Particles.Color_GlowRibbon:LinearColor=&MakeLinearColorFromVectorAndFloat(
     "Vector (RGB)"=&MakeVectorFromLinearColorRGB(Color=&ColorFromCurve(
       ColorCurve=DataInterface() {propertyValues {
         AlphaCurve {keys []}
         BlueCurve {keys [<at 0.1537 0.5521 RCIM_Cubic>,<at 0.5447 0.5>,<at 0.7885 0.7>]}
         GreenCurve {keys [<at 0.1537 0.4885 RCIM_Cubic>,<at 0.5447 0.3879>,<at 0.7885 0.3162>]}
         RedCurve {keys [<at 0.1537 0.1872 RCIM_Cubic>,<at 0.5447 0.1647>,<at 0.7885 0.0164>]}
        }}
       CurveIndex=$Particles.NormalizedRibbonPos
      ))
     "Float (Alpha)"=&FloatFromCurve(
      FloatCurve=DataInterface() {propertyValues {Curve {keys [
          {arriveTangent=0.6799, interpMode=RCIM_Cubic, leaveTangent=0.6799, tangentMode=RCTM_User, time=0, value=0}
          <at 0.4666 1 RCIM_Cubic>
          {arriveTangent=-0.2647, interpMode=RCIM_Cubic, leaveTangent=-0.2647, tangentMode=RCTM_User, time=1, value=0}
         ]}}}
      CurveIndex=$Particles.NormalizedRibbonPos
      Scale Curve=&Multiply_Float(
       A=&FloatFromCurve(
        FloatCurve=DataInterface() {propertyValues {Curve {keys [<at 0 0.5>,<at 0.171 1>,<at 0.7036 1>,<at 0.912 0.5>]}}}
        CurveIndex=Linked() {linkedVariable {name=NPC.FortniteNPC.TimeOfDay_0-1, type=NiagaraFloat}}
       )
       B=&Multiply_Float(
        A=&MakeCustomFloatFromBool(Bool=$User.bIsFrontEndPreview, True Float=0.5, False Float=1)
        B=&Lerp_Float(
         A=&Multiply_Float(
          A=0.5
          B=&Multiply_Float(A=3, B=&Multiply_Float(A=$NPC.FortniteNPC.EventAudio_Amplitude, B=$NPC.FortniteNPC.EventAudio_Amplitude))
         )
         B=&Multiply_Float(
          A=0.6
          B=&Multiply_Float(
           A=0.5
           B=&Multiply_Float(A=3, B=&Multiply_Float(A=$NPC.FortniteNPC.EventAudio_Amplitude, B=$NPC.FortniteNPC.EventAudio_Amplitude))
          )
         )
         Alpha=$System.UsePreview
        )
       )
      )
     )
    )) {isSetParameters=true}
   &ScaleColor(
    Scale Alpha=&Multiply_Float(
     A=&MakeCustomFloatFromBool(Bool=$User.bIsFrontEndPreview, True Float=0.75, False Float=1)
     B=&Lerp_Float(
      A=&Lerp_Float(A=1, B=1.5, Alpha=$NPC.FortniteNPC.EventAudio_Amplitude)
      B=&Lerp_Float(A=1, B=1.5, Alpha=$System.FEFakeReactivity)
      Alpha=$System.UsePreview
     )
    )
    Scale RGB=&Multiply_VectorByFloat(
     Vector=&MakeVectorFromLinearColorRGB(Color=&ColorFromCurve(
       ColorCurve=DataInterface() {propertyValues {
         AlphaCurve {keys []}
         BlueCurve {keys [<at 0.0939 0.7>, {arriveTangent=0.3339, interpMode=RCIM_Cubic, leaveTangent=0.3339, time=0.7241, value=1.25}, <at 0.9923 1>]}
         GreenCurve {keys [<at 0.0939 0.3162>, {arriveTangent=0.6329, interpMode=RCIM_Cubic, leaveTangent=0.6329, time=0.7241, value=1.106}, <at 0.9923 0.8848>]}
         RedCurve {keys [<at 0.0939 0.0164>, {arriveTangent=0.3591, interpMode=RCIM_Cubic, leaveTangent=0.3591, time=0.7241, value=0.4238}, <at 0.9923 0.339>]}
        }}
       CurveIndex=$Particles.NormalizedRibbonPos
      ))
     Float=&Waveform("Waveform [1]"=::NewEnumerator0, "[1] Phase Scale"=2, "[1] Frequency"=0.3, "[1] Amplitude Min/Max"=<40 50>)
    )
    ScaleA=true
    ScaleRGB=true
   ) {module=ScaleColor003}
  ]
  rendererClasses [&NiagaraRibbonRendererProperties]
  renderers [
   {rendererClass=&NiagaraRibbonRendererProperties}
   {rendererClass=&NiagaraRibbonRendererProperties, rendererIndex=1}
   {rendererClass=&NiagaraRibbonRendererProperties, rendererIndex=2}
   {rendererClass=&NiagaraRibbonRendererProperties, rendererIndex=3}
  ]
 }
 Snowflakes {
  emitterUpdateScript [
   &EmitterState(Inactive Response=::NewEnumerator1, Life Cycle Mode=::NewEnumerator1, Loop Duration=1, UseLoopDelay=false)
   &SpawnRate(
    Spawn Probability=0.75
    SpawnRate=&Multiply_Float(A=3.25, B=$System.DistanceSpawn)
    Use Spawn Probability=true
   )
  ]
  particleSpawnScript [
   &FN_SetSpawnGroup()
   &InitializeParticle(
    Lifetime Max=1
    Lifetime Mode=::NewEnumerator1
    Sprite Rotation Angle=&RandomRangeFloat(Minimum=-0.15, Maximum=0.15)
    Sprite Rotation Mode=::NewEnumerator1
    Sprite Size Max=<11 14>
    Sprite Size Min=<8 10>
    Sprite Size Mode=::NewEnumerator2
    UsePositionOffset=false
    Write Material Random=true
   )
   &ShapeLocation(
    Box Midpoint=&SelectVectorFromArray(
     Direct Array Index=&SpawnGroup()
     Vector Selection Array=DataInterface() {propertyValues {FloatData [<0.5 0.2 0.5>, <0.5 0.56 0.5>, <0.5 0.2 0.5>, <0.5 1.5 0.5>]}}
    )
    Box Size=&SelectVectorFromArray(
     Direct Array Index=&SpawnGroup()
     Vector Selection Array=DataInterface() {propertyValues {FloatData [<35 35 5>, <10 38 5>, <35 35 5>, <18 10 5>]}}
    )
    MaskSpawnGroup=false
    Override Seed=false
    Rotation Axis=<0 1 0>
    Rotation Coordinate Space=::Simulation
    Rotation Mode=::NewEnumerator1
    Shape Origin=&VectorFromFloat(Value=0)
    Shape Primitive=::NewEnumerator2
    Surface Only=true
    UseOwnerScale=true
   )
   &CameraOffset(Camera Offset Amount=15)
  ]
  particleUpdateScript [
   &ParticleState()
   &SkeletalMeshLocation(
    Bone / Socket Index=&SelectIntFromArray(Direct Array Index=&SpawnGroup(), Int Selection Array=DataInterface() {propertyValues {IntData [0,1,0,2]}})
    MaskSpawnGroup=false
    Mesh Sampling Type=::NewEnumerator5
    Orientation Sampling=::NewEnumerator1
    Sampled Position Offset:NiagaraPosition=$Particles.ShapeLocation.ShapePosition
    Sampled Position Offset Coordinate Space:ENiagaraSkelMeshTransforms=::NewEnumerator4
    Skeletal Mesh=$System.Skeletal Mesh
    Socket Sampling Mode=::NewEnumerator5
   )
   &SpriteRotationRate(Rotation Rate=&Multiply_Float(
     A=&RandomRangeFloat(Minimum=-10, Maximum=10)
     B=&FloatFromCurve(FloatCurve=DataInterface() {propertyValues {Curve {keys [<at 0 1>,<at 1 0>]}}}, Scale Curve=10)
    ))
   &ScaleSpriteSize(
    Uniform Curve Scale=&Multiply_Float(A=&RandomRangeFloat(Minimum=0.5, Maximum=1), B=1)
    Uniform Curve Sprite Scale=DataInterface() {propertyValues {Curve {keys [{tangentMode=RCTM_User, time=0, value=0.3}, <at 0.1 1.2>, <at 0.2 1>, {tangentMode=RCTM_User, time=1, value=1}]}}}
   )
   &ScaleSpriteSize(
    Uniform Curve Index=&FN_LOD_Distance_Remap(Start Distance=500, End Distance=1500, Minimum Output=0, Maximum Output=1, Clamp Output=true)
    Uniform Curve Sprite Scale=DataInterface() {propertyValues {Curve {keys [{tangentMode=RCTM_User, time=0.4, value=1}, {tangentMode=RCTM_User, time=0.6, value=2}]}}}
   ) {module=ScaleSpriteSize001}
   /SetVariables(Particles.Size002:Vector2f=&Multiply_Vector2D(A=$Particles.SpriteSize, B=&Vector2DFromFloat(Value=1.2))) {isSetParameters=true}
   &DynamicMaterialParameters(
    Index 0 Param 1=&FloatFromCurve(FloatCurve=DataInterface() {propertyValues {Curve {keys [<at 0 -1.5>,<at 0.5 -1.5>,<at 1 1>]}}})
    Index 0 Param 2=0.2
    Module.Index0Param0Write=true
    Module.Index0Param1Write=true
    Module.Index0Param2Write=true
    Module.Index0Param3Write=true
   )
   &ScaleColor(
    Scale Alpha=&MakeCustomFloatFromBool(
     Bool=&SetBoolByFloatComparison(
      A=&Max_Float(
       A:NiagaraFloat=&Lerp_Float(A=0, B=$System.FEFakeReactivity, Alpha=$System.UsePreview)
       B:NiagaraFloat=$NPC.FortniteNPC.EventAudio_Amplitude
      )
      B=0
     )
     True Float=&Lerp_Float(
      A=&Lerp_Float(
       A=&MakeCustomFloatFromBool(
        Bool=$User.bIsFrontEndPreview
        True Float=0.6
        False Float=&FN_PickFloatDependOnPlatform(TargetPlatformActive=1, TargetPlatformNoActive=0.8)
       )
       B=&MakeCustomFloatFromBool(
        Bool=$User.bIsFrontEndPreview
        True Float=1
        False Float=&FN_PickFloatDependOnPlatform(TargetPlatformActive=3, TargetPlatformNoActive=1.5)
       )
       Alpha=$NPC.FortniteNPC.EventAudio_Amplitude
      )
      B=&Lerp_Float(A=0.6, B=1, Alpha=$System.FEFakeReactivity)
      Alpha=$System.UsePreview
     )
     False Float=0
    )
    Scale RGB=&MakeVectorFromLinearColorRGB(Color=&ColorFromCurve(ColorCurve=DataInterface() {propertyValues {
        AlphaCurve {keys []}
        BlueCurve {keys [<at 0.0474 5 RCIM_Cubic>,<at 0.198 1>,<at 0.446 1>,<at 0.8623 0.875 RCIM_Cubic>]}
        GreenCurve {keys [<at 0.0474 2.2571 RCIM_Cubic>,<at 0.198 0.8735>,<at 0.446 0.8369>,<at 0.8623 0.3426 RCIM_Cubic>]}
        RedCurve {keys [<at 0.0474 0.2566 RCIM_Cubic>,<at 0.198 0.6>,<at 0.446 0.4844>,<at 0.8623 0.1257 RCIM_Cubic>]}
       }}))
    ScaleA=true
    ScaleRGB=true
   )
   /SetVariables(Particles.Color_glow:LinearColor=&MakeLinearColorFromVectorAndFloat(
     "Vector (RGB)"=&MakeVectorFromLinearColorRGB(Color=<rgba 0.0858 0.5397 1 1>)
     "Float (Alpha)"=&MakeCustomFloatFromBool(
      Bool=&SetBoolByFloatComparison(
       A=&Max_Float(
        A:NiagaraFloat=&Lerp_Float(A=0, B=$System.FEFakeReactivity, Alpha=$System.UsePreview)
        B:NiagaraFloat=$NPC.FortniteNPC.EventAudio_Amplitude
       )
       B=0
      )
      True Float=&MakeCustomFloatFromBool(
       Bool=$User.bIsFrontEndPreview
       True Float=0.15
       False Float=&Multiply_Float(
        A=0.3
        B=&FloatFromCurve(
         FloatCurve=DataInterface() {propertyValues {Curve {keys [<at 0 0.8>,<at 0.1268 1>,<at 0.6927 1>,<at 1 0.8>]}}}
         CurveIndex=Linked() {linkedVariable {name=NPC.FortniteNPC.TimeOfDay_0-1, type=NiagaraFloat}}
        )
       )
      )
      False Float=0
     )
    )) {isSetParameters=true}
   &SolveForcesAndVelocity(Clamp Velocity=false, Limit Acceleration=false)
  ]
  rendererClasses [&NiagaraSpriteRendererProperties]
  renderers [
   {rendererClass=&NiagaraSpriteRendererProperties}
   {rendererClass=&NiagaraSpriteRendererProperties, rendererIndex=1}
  ]
 }
 Sparkles {
  emitterUpdateScript [
   &EmitterState(Inactive Response=::NewEnumerator1, Life Cycle Mode=::NewEnumerator1, Loop Duration=1, UseLoopDelay=false)
   &SpawnRate(Spawn Probability=0.9, SpawnRate=&Multiply_Float(A=4.5, B=$System.DistanceSpawn), Use Spawn Probability=true)
  ]
  particleSpawnScript [
   &FN_SetSpawnGroup()
   &InitializeParticle(
    Color=<rgba 0.4427 0.8727 1 1>
    Lifetime Max=0.8
    Lifetime Min=0.5
    Lifetime Mode=::NewEnumerator1
    Sprite Rotation Angle=&RandomRangeFloat(Minimum=-0.15, Maximum=0.15)
    Sprite Rotation Mode=::NewEnumerator1
    Sprite Size Mode=::NewEnumerator1
    Uniform Sprite Size Max=11
    Uniform Sprite Size Min=8
    UsePositionOffset=false
    Write Material Random=true
   )
   &ShapeLocation(
    Box Midpoint=&SelectVectorFromArray(
     Direct Array Index=&SpawnGroup()
     Vector Selection Array=DataInterface() {propertyValues {FloatData [<0.5 0.2 0.5>, <0.5 0.56 0.5>, <0.5 0.2 0.5>, <0.5 1.5 0.5>]}}
    )
    Box Size=&SelectVectorFromArray(
     Direct Array Index=&SpawnGroup()
     Vector Selection Array=DataInterface() {propertyValues {FloatData [<35 35 5>, <10 38 5>, <35 35 5>, <15 10 5>]}}
    )
    MaskSpawnGroup=false
    Override Seed=false
    Rotation Axis=<0 1 0>
    Rotation Coordinate Space=::Simulation
    Rotation Mode=::NewEnumerator1
    Shape Origin=&VectorFromFloat(Value=0)
    Shape Primitive=::NewEnumerator2
    Surface Only=true
    UseOwnerScale=true
   )
   &CameraOffset(Camera Offset Amount=15)
  ]
  particleUpdateScript [
   &ParticleState()
   &SkeletalMeshLocation(
    Bone / Socket Index=&SelectIntFromArray(Direct Array Index=&SpawnGroup(), Int Selection Array=DataInterface() {propertyValues {IntData [0,2,0,1]}})
    MaskSpawnGroup=false
    Mesh Sampling Type=::NewEnumerator5
    Orientation Sampling=::NewEnumerator1
    Sampled Position Offset:NiagaraPosition=$Particles.ShapeLocation.ShapePosition
    Sampled Position Offset Coordinate Space:ENiagaraSkelMeshTransforms=::NewEnumerator4
    Skeletal Mesh=$System.Skeletal Mesh
    Socket Sampling Mode=::NewEnumerator5
   ) {module=SkeletalMeshLocation001}
   &SpriteRotationRate(Rotation Rate=&Multiply_Float(
     A=&RandomRangeFloat(Minimum=-10, Maximum=10)
     B=&FloatFromCurve(FloatCurve=DataInterface() {propertyValues {Curve {keys [<at 0 1>,<at 1 0>]}}}, Scale Curve=10)
    ))
   &ScaleSpriteSize(
    Uniform Curve Scale=&RandomRangeFloat(Minimum=0.6, Maximum=1)
    Uniform Curve Sprite Scale=DataInterface() {propertyValues {Curve {keys [{tangentMode=RCTM_User, time=0, value=0.4}, <at 0.1 1.2>, <at 0.2 1>, {tangentMode=RCTM_User, time=1, value=1}]}}}
   )
   /SetVariables(Particles.Size002:Vector2f=&Multiply_Vector2D(
     A=$Particles.SpriteSize
     B=&Vector2DFromFloat(Value=&RandomRangeFloat(Minimum=0.5, Maximum=0.8))
    )) {isSetParameters=true}
   &ScaleColor(
    Scale Alpha=&FloatFromCurve(
     FloatCurve=DataInterface() {propertyValues {Curve {keys [<at 0.5 1>,<at 1 0>]}}}
     Scale Curve=&MakeCustomFloatFromBool(
      Bool=&SetBoolByFloatComparison(
       A=&Max_Float(
        A:NiagaraFloat=&Lerp_Float(A=0, B=$System.FEFakeReactivity, Alpha=$System.UsePreview)
        B:NiagaraFloat=$NPC.FortniteNPC.EventAudio_Amplitude
       )
       B=0
      )
      True Float=&MakeCustomFloatFromBool(Bool=$User.bIsFrontEndPreview, True Float=1, False Float=1.5)
      False Float=0
     )
    )
    ScaleA=true
    ScaleRGB=true
   )
   /SetVariables(Particles.Color003:LinearColor=&MakeLinearColorFromVectorAndFloat(
     "Vector (RGB)"=&MakeVectorFromLinearColorRGB(Color=<rgba 0.0781 0.4051 1 1>)
     "Float (Alpha)"=&FloatFromCurve(
      FloatCurve=DataInterface() {propertyValues {Curve {keys [<at 0 0 RCIM_Cubic>,<at 0.4 1 RCIM_Cubic>,<at 1 0 RCIM_Cubic>]}}}
      Scale Curve=&MakeCustomFloatFromBool(
       Bool=&SetBoolByFloatComparison(
        A=&Max_Float(
         A:NiagaraFloat=&Lerp_Float(A=0, B=$System.FEFakeReactivity, Alpha=$System.UsePreview)
         B:NiagaraFloat=$NPC.FortniteNPC.EventAudio_Amplitude
        )
        B=0
       )
       True Float=&MakeCustomFloatFromBool(Bool=$User.bIsFrontEndPreview, True Float=0.75, False Float=1)
       False Float=0
      )
     )
    )) {isSetParameters=true}
   /SetVariables(Particles.Size003:Vector2f=&Multiply_Vector2D(A=<10 10>, B=&Vector2DFromFloat(Value=1))) {isSetParameters=true}
   &SolveForcesAndVelocity(Clamp Velocity=false, Limit Acceleration=false)
   &SubUVAnimation(
    End Frame Range Override=3
    "Sprite Renderer "=DataInterface() {propertyValues {SpriteRenderer=&self:Sparkles_0.NiagaraSpriteRendererProperties_0}}
    UseEndFrame=true
    UseStartFrame=true
   )
  ]
  rendererClasses [&NiagaraSpriteRendererProperties]
  renderers [
   {rendererClass=&NiagaraSpriteRendererProperties}
   {rendererClass=&NiagaraSpriteRendererProperties, rendererIndex=1}
  ]
 }
}
summary {
 systemName=NS_Guitar_FlossYawn_IdleFX
 emitters [
  {emitterName=NE_SpiralRibbon, rendererClasses [&NiagaraRibbonRendererProperties]}
  {emitterName=Snowflakes, rendererClasses [&NiagaraSpriteRendererProperties]}
  {emitterName=Sparkles, rendererClasses [&NiagaraSpriteRendererProperties]}
  {emitterName=Line, rendererClasses [&NiagaraSpriteRendererProperties]}
  {emitterName=Glow, rendererClasses [&NiagaraSpriteRendererProperties, &NiagaraMeshRendererProperties]}
  {bEnabled=false, emitterName=DebugValues, rendererClasses [&NiagaraSpriteRendererProperties]}
 ]
 userVariables [User.bIsFrontEndPreview {type=NiagaraBool}]
}
